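# Base UIKit text input field built on UIKitModularButton, with optional background, icon and expandable width.
# PreConstruct creates and registers visual blocks, then applies text, size and style settings.
# Text edits enforce a character limit, update the icon material's hint parameter, and broadcast unless trimmed.
# Focus and non-touch hover drive block transitions and AnimExpand, whose WidthAnimation value lerps SizeBox max width, collapsing only when text is empty.
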WidgetBlueprint /FortUILibrary/Systems/InputField/WBP_UIKit_InputField_Base
animations [AnimExpand {
  binding_count=1
  length_seconds=1
  object=&self:AnimExpand
  sequence {bindings [WBP_UIKit_InputField_Base_C_26 {
     id=#1
     possessed_class=WBP_UIKit_InputField_Base_C
     tracks [{
       type=MovieSceneDoubleTrack
       display_name=Width Animation
       sections [{type=MovieSceneDoubleSection, end_frame=20, start_frame=0, channels [None {type=MovieSceneScriptingDoubleChannel, keys [[0,0],[20,1]]}]}]
      }]
    }]}
 }]
graphs [
 AddBackground {nodes [
   AddBackground() {is=title K2Node_FunctionEntry}
   > Cleanup!(execute, self, Widget:Widget Object Reference=Get Block_Background() -> (Block_Background:WBP UIKit Block Background Base Object Reference)) -> (then)
   > Branch(Condition=Get UseBackground() -> (UseBackground:bool))
   > n1 = Create Widget!(
    execute:Exec
    Class:User Widget Class Reference=Get BackgroundClass() -> (BackgroundClass:WBP UIKit Block Background Base Class Reference)
    OwningPlayer:Player Controller Object Reference
   ) -> (then:Exec, ReturnValue:WBP UIKit Block Background Base Object Reference)
   > n2 = Set Block_Background(Block_Background:WBP UIKit Block Background Base Object Reference=@n1) -> (Output_Get:WBP UIKit Block Background Base Object Reference)
   > Branch(Condition=IsValid(Object=@n2))
   > AddChildToGrid!(
    execute
    self:Grid Panel Object Reference=Get Main() -> (Main:Grid Panel Object Reference)
    Content:Widget Object Reference=@n2
    InRow:int=0
    InColumn:int=0
   ) -> (then, ReturnValue:Grid Slot Object Reference)
   > RegisterModularBlock!(execute, self:UIKit Modular Button Object Reference, Block:UIKit Block Object Reference=@n2) -> (then)
   > SetBackgroundMaterial!(
    execute
    self:WBP UIKit Block Background Base Object Reference=@n2
    Material Instance:"Material Instance Object Reference (by ref)"=Get BackgroundMaterial() -> (BackgroundMaterial:Material Instance Object Reference)
   ) -> (then)
   > n3 = Cast To WBP_UIKit_Block_InputField_Background!(execute, Object=@n2) -> (then, CastFailed, AsWBP UIKit Block Input Field Background:WBP UIKit Block Input Field Background Object Reference) {is=title K2Node_DynamicCast}
   > SetCornerRadius!(
    execute
    self:WBP UIKit Block Input Field Background Object Reference=@n3
    Radius in Pixels:bool=Get UsePixels() -> (UsePixels:bool)
    TopLeft:double=Get TopLeft() -> (TopLeft:double)
    TopRight:double=Get TopRight() -> (TopRight:double)
    BottomRight:double=Get BottomRight() -> (BottomRight:double)
    BottomLeft:double=Get BottomLeft() -> (BottomLeft:double)
   ) -> (then)
  ]}
 AddIcon {nodes [
   AddIcon() {is=title K2Node_FunctionEntry}
   > Cleanup(Widget=Get Block_Icon() -> (Block_Icon:WBP UIKit Block Image Object Reference))
   > Branch(Condition=Get UseIcon() -> (UseIcon:bool))
   > n1 = Create Widget(Class=Get IconClass() -> (IconClass:WBP UIKit Block Image Class Reference)) -> (ReturnValue:WBP UIKit Block Image Object Reference)
   > n2 = Set Block_Icon(Block_Icon:WBP UIKit Block Image Object Reference=@n1) -> (Output_Get:WBP UIKit Block Image Object Reference)
   > Branch(Condition=IsValid(Object=@n2))
   > AddChildToGrid(self=Get Content() -> (Content:Grid Panel Object Reference), Content=@n2, InRow=1, InColumn=0)
   > RegisterModularBlock(Block=@n2)
  ]}
 Cleanup {nodes [
   n1 = Cleanup!() -> (then, Widget:Widget Object Reference) {is=title K2Node_FunctionEntry}
   > Branch(Condition=IsValid(Object=@n1))
   > RemoveChild(self=GetParent!(self:Widget Object Reference=@n1) -> (ReturnValue:Panel Widget Object Reference), Content=@n1)
  ]}
 EventGraph {nodes [
   "On Input Icon Updated (InputActionWidget)"() {is=title K2Node_ComponentBoundEvent}
   > SetInputIconDarkMode!(execute, self, UseDark:bool=Get UseDarkInputIcon() -> (UseDarkInputIcon:bool)) -> (then)
   n1 = OnTextChanged_Event!() -> (OutputDelegate, then, Text:"Text (by ref)") {is=title K2Node_CustomEvent}
   > Set EditableText(EditableText:Text=@n1.Text) -> (Output_Get:Text)
   > Is Text Empty!(execute, self, IsEmpty:bool=TextIsEmpty(InText=@n1.Text)) -> (then)
   > n2 = SetMaxTextCharacters!(execute, self) -> (then, IsTextTrimmed:bool)
   > n3 = Branch(Condition=@n2)
   Call On Text Changed!(execute:Exec=@n3.else, self:WBP UIKit Input Field Base Object Reference, Text:Text=@n1.Text) -> (then:Exec) {is=title K2Node_CallDelegate}
   n4 = OnTextCommitted_Event!() -> (OutputDelegate, then, Text:"Text (by ref)", CommitMethod:ETextCommit Enum=Default) {is=title K2Node_CustomEvent}
   > Call On Text Commited!(execute:Exec, self:WBP UIKit Input Field Base Object Reference, Text:Text=@n4.Text, Commit method:ETextCommit Enum=@n4.CommitMethod) -> (then:Exec) {is=title K2Node_CallDelegate}
   n5 = Event PreConstruct()
   > n6 = Sequence!(execute) -> (then_0, then_1, then_2:Exec, then_3:Exec, then_4:Exec)
   > Branch(Condition=@n5.IsDesignTime)
   > PresetForDesignTime()
   n7 = For Each Loop(
    Exec=@n6.then_1
    Array:Array of UIKit Block Object References=Make Array!(
     "[0]":UIKit Block Object Reference=Get Block_EditableText() -> (Block_EditableText:WBP UIKit Block Editable Text Object Reference)
     "[1]":UIKit Block Object Reference=Get Block_Outline() -> (Block_Outline:WBP UIKit Block Outline Object Reference)
    ) -> (Array:Array of UIKit Block Object References)
   ) -> (Array Element:UIKit Block Object Reference)
   > RegisterModularBlock(Block=@n7)
   SetEditableText!(execute=[@n14.then, @n6.then_2], self, SetTextFields:bool=true) -> (then)
   > SetClearKeyboardFocusOnCommitBehavior!(execute, self, ShouldClearKeyboardFocusOnCommit:bool=Get ShouldClearKeyboardFocusOnCommit() -> (ShouldClearKeyboardFocusOnCommit:bool)) -> (then)
   > SetEnableIntegratedKeyboard!(execute, self, EnableIntegratedKeyboard:bool=Get MobileEnableIntegratedKeyboard() -> (MobileEnableIntegratedKeyboard:bool)) -> (then)
   > AddBackground() {is=title K2Node_CallFunction}
   > AddIcon() {is=title K2Node_CallFunction}
   > SetIconProperties!(execute, self, Data:S UI Image Structure=Get IconProperties() -> (IconProperties:S UI Image Structure)) -> (then)
   > n8 = Branch(Condition=Get bExpandable() -> (bExpandable:bool))
   > SetExpandableSize()
   SetSize!(
    execute=@n8.else
    self
    WIdth:S UI Bool Float Structure=Get WidthOverride() -> (WidthOverride:S UI Bool Float Structure)
    Height:S UI Bool Float Structure=Get HeightOverride() -> (HeightOverride:S UI Bool Float Structure)
   ) -> (then)
   SetBlocksTiming!(
    execute=@n6.then_3
    self:UIKit Modular Button Object Reference
    Timing_Hovering:float=0.2
    Timing_Unhovering:float=0.1
    Timing_Focusing:float=0.3
    Timing_Unfocusing:float=0.15
    Timing_Pressing:float=0
    Timing_Releasing:float=0
    Timing_Disabling:float=0.1
    Timing_Enabling:float=0.1
    Timing_Selecting:float=0
    Timing_DeselectingIdle:float=0
    Timing_DeselectingFocused:float=0
   ) -> (then)
   > DelayUntilNextTick!(execute) -> (then)
   n9 = Get EditableText!(self:WBP UIKit Block Editable Text Object Reference=Get Block_EditableText()) -> (EditableText:UIKit Editable Text Object Reference)
   > Is Text Empty(IsEmpty=TextIsEmpty(InText=Get Text(self:Editable Text Object Reference=@n9)))
   n10 = GetConsoleVariableBoolValue!(execute=@n6.then_4, VariableName:String=UI.InputField.HideInputAction) -> (then, ReturnValue:bool)
   > SetHideInputAction!(execute, self:Common Button Base Object Reference, bInHideInputAction:bool=@n10) -> (then)
   Event On Pressed() {is=title K2Node_Event}
   > RequestTransitionOnBlocks!(execute, self:UIKit Modular Button Object Reference, State:EUIKitBlockVisualState Enum=Pressed) -> (then)
   Event OnRemovedFromFocusPath!() -> (OutputDelegate, then, InFocusEvent:Focus Event Structure) {is=title K2Node_Event}
   > RequestTransitionOnBlocks(State=Unfocused)
   > ExpandField!(execute, self, Expand:bool=false, FocusPath:bool=true) -> (then)
   Event On Released() {is=title K2Node_Event}
   > RequestTransitionOnBlocks(State=Released)
   Event OnInitialized()
   n11 = Get EditableText(self=Get Block_EditableText())
   > "Bind Event to OnTextChanged (Editable Text)"(self:Editable Text Object Reference=@n11, Delegate=@n1) {is=title K2Node_AddDelegate}
   > "Bind Event to OnTextCommitted (Editable Text)"(self:Editable Text Object Reference=@n11, Delegate=@n4) {is=title K2Node_AddDelegate}
   Event On Hovered()
   > n12 = Branch(Condition=IsUsingTouch(PlayerController=GetOwningPlayer()))
   RequestTransitionOnBlocks(execute=@n12.else, State=Focused)
   > ExpandField(Expand=true, FocusPath=false)
   Event On Unhovered()
   > n13 = Branch(Condition=IsUsingTouch(PlayerController=GetOwningPlayer()))
   RequestTransitionOnBlocks(execute=@n13.else, State=Unfocused)
   > ExpandField(Expand=false, FocusPath=false)
   n14 = Event On Widget Styled() {is=title K2Node_Event}
   Event OnAddedToFocusPath!() -> (OutputDelegate, then, InFocusEvent:Focus Event Structure) {is=title K2Node_Event}
   > RequestTransitionOnBlocks(State=Focused)
   > ExpandField(Expand=true, FocusPath=true)
  ]}
 ExpandField {nodes [
   n1 = ExpandField!() -> (then, Expand:bool=~, FocusPath:bool=~) {is=title K2Node_FunctionEntry}
   > Branch(Condition=Get bExpandable())
   > n2 = Branch(Condition=@n1)
   > Branch(Condition=NOT Boolean(A=Get bIsExpanded() -> (bIsExpanded:bool)))
   > PlayAnimationForward(InAnimation=Get AnimExpand() -> (AnimExpand:Widget Animation Object Reference), PlaybackSpeed=3)
   > n3 = Set bIsExpanded(bIsExpanded:bool=true) -> (Output_Get:bool)
   > Call Expand State Changed!(execute:Exec, self:WBP UIKit Input Field Base Object Reference, IsExpanded:bool=@n3) -> (then:Exec) {is=title K2Node_CallDelegate}
   n4 = Get EditableText!() -> (EditableText:Text)
   n5 = Get Block_EditableText()
   n6 = NOT Boolean(A=HasAnyUserFocus!(self:Widget Object Reference=Get EditableText!(self:WBP UIKit Block Editable Text Object Reference=@n5) -> (EditableText:UIKit Editable Text Object Reference)) -> (ReturnValue:bool))
   n7 = Get FocusPath() -> (FocusPath:bool)
   Branch(execute=@n2.else, Condition=AND Boolean!(A=TextIsEmpty(InText=@n4), B=Get bIsExpanded(), C:bool=OR Boolean(A=@n6, B=@n7)) -> (ReturnValue))
   > PlayAnimationReverse(InAnimation=Get AnimExpand(), PlaybackSpeed=3)
   > n8 = Set bIsExpanded(bIsExpanded=false)
   > Call Expand State Changed(IsExpanded=@n8)
  ]}
 ExpandStateChanged {nodes [ExpandStateChanged!() -> (then, IsExpanded:bool=~) {is=title K2Node_FunctionEntry}]}
 Is Text Empty {nodes [
   n1 = Is Text Empty!() -> (then, IsEmpty:bool=~) {is=title K2Node_FunctionEntry}
   > n2 = Get!(execute:Exec) -> (then:Exec, else:Exec, Block_Icon:WBP UIKit Block Image Object Reference)
   > SetScalarParameterValue(
    self=Get ImageMID!(self:WBP UIKit Block Image Object Reference=@n2) -> (ImageMID:Material Instance Dynamic Object Reference)
    ParameterName=Trans_HintText
    Value=Select(Option 0:float=0, Option 1:float=1, Index=@n1) -> (ReturnValue:float)
   )
  ]}
 OnFocusReceived {nodes [
   OnFocusReceived!() -> (then, MyGeometry:Geometry Structure, InFocusEvent:Focus Event Structure) {is=title K2Node_FunctionEntry}
   > ExpandField(Expand=true, FocusPath=false) {is=title K2Node_CallFunction}
   > Return Node!(execute, ReturnValue:Event Reply Structure=Handled())
  ]}
 OnTextChanged {nodes [OnTextChanged!() -> (then, Text:Text) {is=title K2Node_FunctionEntry}]}
 OnTextCommited {nodes [OnTextCommited!() -> (then, Text:Text, Commit method:ETextCommit Enum=Default) {is=title K2Node_FunctionEntry}]}
 PresetForDesignTime {nodes [
   PresetForDesignTime() {is=title K2Node_FunctionEntry}
   n1 = Get Slot_Main() -> (Slot_Main:Named Slot Object Reference)
   > SetVisibility(
    self=@n1
    InVisibility=Select(
     Option 0:ESlateVisibility Enum=Collapsed
     Option 1:ESlateVisibility Enum=SelfHitTestInvisible
     Index=HasAnyChildren!(self:Panel Widget Object Reference=@n1) -> (ReturnValue:bool)
    ) -> (ReturnValue:ESlateVisibility Enum)
   )
  ]}
 SetClearKeyboardFocusOnCommitBehavior {nodes [
   n1 = SetClearKeyboardFocusOnCommitBehavior!() -> (then, ShouldClearKeyboardFocusOnCommit:bool=~) {is=title K2Node_FunctionEntry}
   > n2 = Set ShouldClearKeyboardFocusOnCommit(ShouldClearKeyboardFocusOnCommit:bool=@n1) -> (Output_Get:bool)
   > SetClearKeyboardFocusOnCommitBehavior(self:WBP UIKit Block Editable Text Object Reference=Get Block_EditableText(), ShouldClearKeyboardFocusOnCommit=@n2) {is=title K2Node_CallFunction}
  ]}
 SetEditableText {nodes [
   n1 = SetEditableText!() -> (then, SetTextFields:bool=~) {is=title K2Node_FunctionEntry}
   > n2 = Sequence!(execute) -> (then_0, then_1)
   > Branch(Condition=@n1)
   n3 = Get Block_EditableText()
   > SetText!(execute, self:WBP UIKit Block Editable Text Object Reference=@n3, Text:Text=Get Text!() -> (Text)) -> (then)
   > SetHintText!(execute, self:WBP UIKit Block Editable Text Object Reference=@n3, Hint Text:Text=Get Hint Text() -> (Hint Text:Text)) -> (then)
   SetFont!(
    execute=@n2.then_1
    self:WBP UIKit Block Editable Text Object Reference=@n3
    Font:Slate Font Info Structure=Get Font() -> (Font:Slate Font Info Structure)
   ) -> (then)
   > SetIsPassword!(execute, self:WBP UIKit Block Editable Text Object Reference=@n3, IsPassword:bool=Get IsPassword() -> (IsPassword:bool)) -> (then)
   > SetReadOnly!(execute, self:WBP UIKit Block Editable Text Object Reference=@n3, IsReadOnly:bool=Get IsReadOnly() -> (IsReadOnly:bool)) -> (then)
   > SetJustification!(
    execute
    self:WBP UIKit Block Editable Text Object Reference=@n3
    Justification:ETextJustify Enum=Get TextJustification() -> (TextJustification:ETextJustify Enum=Left)
   ) -> (then)
   > SetPadding!(execute, self:User Widget Object Reference=@n3, InPadding:Margin Structure=Get TextPadding() -> (TextPadding:Margin Structure)) -> (then)
   > SetMobileFontSizeMultiplier!(
    execute
    self:WBP UIKit Block Editable Text Object Reference=@n3
    MobileFontSizeMultiplier:float=Get Mobile Text Size Multiplier() -> (Mobile Text Size Multiplier:double)
   ) -> (then)
  ]}
 SetEnableIntegratedKeyboard {nodes [
   n1 = SetEnableIntegratedKeyboard!() -> (then, EnableIntegratedKeyboard:bool=~) {is=title K2Node_FunctionEntry}
   > SetEnableIntegratedKeyboard(self:WBP UIKit Block Editable Text Object Reference=Get Block_EditableText(), EnableIntegratedKeyboard=@n1) {is=title K2Node_CallFunction}
  ]}
 SetExpandableSize {nodes [
   SetExpandableSize() {is=title K2Node_FunctionEntry}
   n1 = Get SIzeBox() -> (SIzeBox:Size Box Object Reference)
   > ClearWidthOverride(self:Size Box Object Reference=@n1)
   n2 = Break S UI Bool Float!(S_UI_BoolFloat:"S UI Bool Float Structure (by ref)"=Get WidthOverride()) -> (Enable_3_#2:bool=~, Value_4_#3:double=~) {is=title K2Node_BreakStruct}
   n3 = Select(Option 0:float=56, Option 1:float=@n2.Value_4_#3, Index=@n2) -> (ReturnValue:float)
   > SetMinDesiredWidth!(execute, self:Size Box Object Reference=@n1, InMinDesiredWidth:float=@n3) -> (then)
   > SetMaxDesiredWidth!(execute, self:Size Box Object Reference=@n1, InMaxDesiredWidth:float=@n3) -> (then)
   n4 = Get IconProperties()
   n5 = SlotAsGridSlot(Widget=Get Block_Icon())
   > SetPadding(
    self:Grid Slot Object Reference=@n5
    InPadding=Break S UI Placement!(S_UI_Placement:"S UI Placement Structure (by ref)"=Break S UI Image!(S_UI_Image:"S UI Image Structure (by ref)"=@n4) -> (
      Placement_67_#4:S UI Placement Structure
      ImageSize_72_#5:Vector 2D Structure
      Material_75_#6:Material Instance Object Reference
     ) {is=title K2Node_BreakStruct}) -> (
     HorizontalAlignment_4_#4:EHorizontalAlignment Enum
     VerticalAlignment_5_#7:EVerticalAlignment Enum
     Padding_8_#8:Margin Structure
     Grid_20_#9:S UI Grid Cell Structure
    ) {is=title K2Node_BreakStruct}.Padding_8_#8
   )
   > SetColumn!(execute, self:Grid Slot Object Reference=@n5, InColumn:int=3) -> (then)
   > SetSize!(execute, self:Spacer Object Reference=Get Spacer() -> (Spacer:Spacer Object Reference), InSize:Vector 2D Structure=(X=1920,Y=0)) -> (then)
   > n6 = AddChildToGrid(self=Get Content(), Content=Get InputActionWidget() -> (InputActionWidget:Common Action Widget Object Reference), InRow=1, InColumn=2)
   > SetPadding!(execute, self=@n6, InPadding_Left:float=0, InPadding_Top:float=0, InPadding_Right:float=6, InPadding_Bottom:float=0) -> (then)
   n7 = Break S UI Bool Float(S_UI_BoolFloat=Get HeightOverride()) -> (Enable_3_#2=~, Value_4_#3=~)
   > Branch(Condition=@n7)
   > SetHeightOverride(self=Get SIzeBox(), InHeightOverride=@n7.Value_4_#3)
  ]}
 SetGridAlignment {nodes [
   n1 = SetGridAlignment!() -> (then, WIdget:Widget Object Reference, Horizontal:EHorizontalAlignment Enum=HAlign_Fill, Vertical:EVerticalAlignment Enum=VAlign_Fill) {is=title K2Node_FunctionEntry}
   > n2 = Cast To GridSlot!(execute, Object=Get Slot!(self:Widget Object Reference=@n1) -> (Slot:Panel Slot Object Reference)) -> (then, CastFailed, AsGrid Slot:Grid Slot Object Reference) {is=title K2Node_DynamicCast}
   > SetHorizontalAlignment(self:Grid Slot Object Reference=@n2, InHorizontalAlignment=@n1.Horizontal)
   > SetVerticalAlignment(self:Grid Slot Object Reference=@n2, InVerticalAlignment=@n1.Vertical)
  ]}
 SetGridColumnAndRow {nodes [
   n1 = SetGridColumnAndRow!() -> (then, WIdget:Widget Object Reference, Row:int, Row Span:int, Column:int, Column Span:int) {is=title K2Node_FunctionEntry}
   > n2 = Cast To GridSlot(Object=Get Slot(self=@n1))
   > SetColumn(self=@n2, InColumn=@n1.Column)
   > SetColumnSpan!(execute, self:Grid Slot Object Reference=@n2, InColumnSpan:int=@n1.Column Span) -> (then)
   > SetRow!(execute, self:Grid Slot Object Reference=@n2, InRow:int=@n1.Row) -> (then)
   > SetRowSpan!(execute, self:Grid Slot Object Reference=@n2, InRowSpan:int=@n1.Row Span) -> (then)
  ]}
 SetGridLayer {nodes [
   n1 = SetGridLayer!() -> (then, Widget:Widget Object Reference, Layer:int) {is=title K2Node_FunctionEntry}
   > n2 = Cast To GridSlot(Object=Get Slot(self=@n1))
   > SetLayer!(execute, self:Grid Slot Object Reference=@n2, InLayer:int=@n1.Layer) -> (then)
  ]}
 SetGridNudge {nodes [
   n1 = SetGridNudge!() -> (then, WIdget:Widget Object Reference, Nudge:Vector 2D Structure) {is=title K2Node_FunctionEntry}
   > n2 = Cast To GridSlot(Object=Get Slot(self=@n1))
   > SetNudge!(execute, self:Grid Slot Object Reference=@n2, InNudge:Vector 2D Structure=@n1.Nudge) -> (then)
  ]}
 SetGridPadding {nodes [
   n1 = SetGridPadding!() -> (then, WIdget:Widget Object Reference, Padding:Margin Structure) {is=title K2Node_FunctionEntry}
   > n2 = Cast To GridSlot(Object=Get Slot(self=@n1))
   > SetPadding!(execute, self=@n2, InPadding:Margin Structure=@n1.Padding) -> (then)
  ]}
 SetHintText {nodes [
   n1 = SetHintText!() -> (then, Hint Text:Text) {is=title K2Node_FunctionEntry}
   > n2 = Set Hint Text(Hint Text:Text=@n1) -> (Output_Get:Text)
   > SetHintText(self=Get Block_EditableText(), Hint Text=@n2) {is=title K2Node_CallFunction}
  ]}
 SetIcon {nodes [
   n1 = SetIcon!() -> (then, Icon:Texture 2D Object Reference) {is=title K2Node_FunctionEntry}
   > n2 = Set Icon(Icon:Texture 2D Object Reference=@n1) -> (Output_Get:Texture 2D Object Reference)
   n3 = Get Block_Icon()
   > Branch(Condition=IsValid(Object=@n3))
   > SetTexture!(execute, self:WBP UIKit Block Image Object Reference=@n3, InTexture:"Texture 2D Object Reference (by ref)"=@n2) -> (then)
  ]}
 SetIconProperties {nodes [
   n1 = SetIconProperties!() -> (then, Data:S UI Image Structure) {is=title K2Node_FunctionEntry}
   n2 = Get Block_Icon()
   > Branch(Condition=IsValid(Object=@n2))
   n3 = Break S UI Image(S_UI_Image=@n1)
   > SetPlacement!(execute, self, Widget:Widget Object Reference=@n2, Placement:S UI Placement Structure=@n3) -> (then)
   > SetSize!(execute, self:WBP UIKit Block Image Object Reference=@n2, Desired Size:Vector 2D Structure=@n3.ImageSize_72_#5) -> (then)
   > SetMaterial!(
    execute
    self:WBP UIKit Block Image Object Reference=@n2
    InMaterial:"Material Instance Object Reference (by ref)"=@n3.Material_75_#6
   ) -> (then)
   > SetTexture(self=@n2, InTexture=Get Icon() -> (Icon:Texture 2D Object Reference))
  ]}
 SetInputIconDarkMode {nodes [
   n1 = SetInputIconDarkMode!() -> (then, UseDark:bool=~) {is=title K2Node_FunctionEntry}
   > Set UseDarkInputIcon(UseDarkInputIcon:bool=@n1) -> (Output_Get:bool)
   > n2 = GetIconDynamicMaterial!(execute, self:Common Action Widget Object Reference=Get InputActionWidget()) -> (then, ReturnValue:Material Instance Dynamic Object Reference)
   > n3 = Set CurrentDynamicPromptIcon(CurrentDynamicPromptIcon:Material Instance Dynamic Object Reference=@n2) -> (Output_Get:Material Instance Dynamic Object Reference)
   > SetScalarParameterValue(self=@n3, ParameterName=Use Dark, Value=Select(Option 0=0, Option 1=1, Index=Get UseDarkInputIcon()))
   > SetScalarParameterValue(
    self=Get CurrentDynamicPromptIcon() -> (CurrentDynamicPromptIcon:Material Instance Dynamic Object Reference)
    ParameterName=Outline Opacity
    Value=Select(Option 0=0.5, Option 1=0, Index=Get UseDarkInputIcon())
   )
  ]}
 SetMaxTextCharacters {nodes [
   SetMaxTextCharacters() {is=title K2Node_FunctionEntry}
   > n1 = LimitTextLength!(
    execute
    EditableTextWidget:Widget Object Reference=Get EditableText(self=Get Block_EditableText())
    Length:int=Get MaxTextCharacters() -> (MaxTextCharacters:int=0)
   ) -> (then, WasTrimmed:bool)
   > Return Node!(execute, IsTextTrimmed:bool=@n1)
  ]}
 SetPlacement {nodes [
   n1 = SetPlacement!() -> (then, Widget:Widget Object Reference, Placement:S UI Placement Structure) {is=title K2Node_FunctionEntry}
   n2 = Break S UI Placement(S_UI_Placement=@n1.Placement)
   > SetGridAlignment!(
    execute
    self
    WIdget:Widget Object Reference=@n1
    Horizontal:EHorizontalAlignment Enum=@n2
    Vertical:EVerticalAlignment Enum=@n2.VerticalAlignment_5_#7
   ) -> (then) {is=title K2Node_CallFunction}
   > SetGridPadding!(execute, self, WIdget:Widget Object Reference=@n1, Padding:Margin Structure=@n2.Padding_8_#8) -> (then) {is=title K2Node_CallFunction}
   n3 = Break S UI Grid Cell!(S_UI_GridCell:"S UI Grid Cell Structure (by ref)"=@n2.Grid_20_#9) -> (
    Row_15_#4:int
    RowSpan_17_#10:int
    Column_19_#11:int
    ColumnSpan_21_#12:int
    Layer_23_#13:int
    Nudge_26_#14:Vector 2D Structure
   ) {is=title K2Node_BreakStruct}
   > SetGridColumnAndRow!(
    execute
    self
    WIdget:Widget Object Reference=@n1
    Row:int=@n3
    Row Span:int=@n3.RowSpan_17_#10
    Column:int=@n3.Column_19_#11
    Column Span:int=@n3.ColumnSpan_21_#12
   ) -> (then) {is=title K2Node_CallFunction}
   > SetGridLayer!(execute, self, Widget:Widget Object Reference=@n1, Layer:int=@n3.Layer_23_#13) -> (then) {is=title K2Node_CallFunction}
   > SetGridNudge!(execute, self, WIdget:Widget Object Reference=@n1, Nudge:Vector 2D Structure=@n3.Nudge_26_#14) -> (then) {is=title K2Node_CallFunction}
  ]}
 SetSize {nodes [
   n1 = SetSize!() -> (then, WIdth:S UI Bool Float Structure, Height:S UI Bool Float Structure) {is=title K2Node_FunctionEntry}
   > n2 = Sequence()
   n3 = Break S UI Bool Float(S_UI_BoolFloat=@n1) -> (Enable_3_#2=~, Value_4_#3=~)
   > Branch(Condition=@n3)
   n4 = Get SIzeBox()
   > SetWidthOverride(self=@n4, InWidthOverride=@n3.Value_4_#3)
   n5 = Break S UI Bool Float(S_UI_BoolFloat=@n1.Height) -> (Enable_3_#2=~, Value_4_#3=~)
   Branch(execute=@n2.then_1, Condition=@n5)
   > SetHeightOverride(self=@n4, InHeightOverride=@n5.Value_4_#3)
  ]}
 SetText {nodes [
   n1 = SetText!() -> (then, Text:Text) {is=title K2Node_FunctionEntry}
   > n2 = Set Text!(execute, Text=@n1) -> (then, Output_Get)
   > SetText(self=Get Block_EditableText(), Text=@n2) {is=title K2Node_CallFunction}
  ]}
 SetWidthAnimation {nodes [
   n1 = SetWidthAnimation!() -> (then, WidthAnimation:double=~) {is=title K2Node_FunctionEntry}
   n2 = Break S UI Bool Float(S_UI_BoolFloat=Get MaxExpandedWidth() -> (MaxExpandedWidth:S UI Bool Float Structure)) -> (Enable_3_#2=~, Value_4_#3=~)
   n3 = Break S UI Bool Float(S_UI_BoolFloat=Get WidthOverride()) -> (Enable_3_#2=~, Value_4_#3=~)
   > SetMaxDesiredWidth(
    self=Get SIzeBox()
    InMaxDesiredWidth=Lerp(
     A=Select(Option 0:double=56, Option 1:double=@n3.Value_4_#3, Index=@n3) -> (ReturnValue:double)
     B=Select(Option 0=1920, Option 1=@n2.Value_4_#3, Index=@n2)
     Alpha=@n1
    )
   )
  ]}
 UpdateIconVisibility {nodes [
   n1 = UpdateIconVisibility!() -> (then, IsVisible:bool=~) {is=title K2Node_FunctionEntry}
   > Branch(Condition=IsValid(Object=Get Block_Icon()))
   > SetVisibility(
    self=Get Block_Icon()
    InVisibility=Select(Option 0:ESlateVisibility Enum=Collapsed, Option 1:ESlateVisibility Enum=SelfHitTestInvisible, Index=@n1) -> (ReturnValue:ESlateVisibility Enum)
   )
  ]}
]
variables [
 BackgroundClass {type=Class, default=/FortUILibrary/Systems/InputField/Blocks/WBP_UIKit_Block_InputField_Background.WBP_UIKit_Block_InputField_Background_C}
 BackgroundMaterial {type=MaterialInstance, default=/FortUILibrary/Systems/InputField/MaterialInstances/MI_UIKit_InputField_Background}
 bExpandable {type=boolean}
 bIsExpanded {type=boolean}
 Block_Background {type=WBP_UIKit_BlockBackground_Base_C, default=null}
 Block_Icon {type=WBP_UIKit_Block_Image_C, default=null}
 BottomLeft {type=number, default=20}
 BottomRight {type=number, default=20}
 EditableText {type=string, default=""}
 ExpandStateChanged {}
 Font {
  type=SlateFontInfo
  default {
   bForceMonospaced=false
   bMaterialIsStencil=false
   fontMaterial=/FortUILibrary/Systems/InputField/MaterialInstances/MI_UIKit_InputFieldText
   fontObject=&HeadingNow
   letterSpacing=0
   monospacedWidth=1
   size=18
   skewAmount=0
   typefaceFontName=74 Regular
   outlineSettings {
    bApplyOutlineToDropShadows=false
    bMiteredCorners=false
    bSeparateFillAlpha=false
    outlineMaterial=null
    outlineSize=0
    outlineColor=<rgba 0 0 0 1>
   }
  }
 }
 HeightOverride {type=S_UI_BoolFloat, default {enable=false, value=0}}
 Hint Text {type=string, default=Hint...}
 Icon {type=Texture2D, default=/Game/Athena/UI/InGame/Sidekick/Reactions/100}
 IconClass {type=Class, default=/FortUILibrary/Systems/ModularButton/Blocks/WBP_UIKit_Block_Image.WBP_UIKit_Block_Image_C}
 IconProperties {
  type=S_UI_Image
  default {
   material=/FortUILibrary/Systems/InputField/MaterialInstances/MI_UIKit_InputFieldIcon
   image Size=<32 32>
   placement {
    horizontal Alignment=HAlign_Center
    vertical Alignment=VAlign_Center
    grid {column=0, column Span=0, layer=0, row=1, row Span=0, nudge=<0 0>}
    padding=<ltrb 0 0 0 0>
   }
  }
 }
 IsPassword {type=boolean}
 IsReadOnly {type=boolean}
 MaxExpandedWidth {type=S_UI_BoolFloat, default {enable=false, value=1920}}
 MaxTextCharacters {type=integer, default=40}
 Mobile Text Size Multiplier {type=number, default=1}
 MobileEnableIntegratedKeyboard {type=boolean}
 OnTextChanged {}
 OnTextCommited {}
 ShouldClearKeyboardFocusOnCommit {type=boolean, default=true}
 Text {type=string, default=""}
 TextJustification {type=ETextJustify, default=Left}
 TextPadding {type=Margin, default=<ltrb 16 0 16 0>}
 TopLeft {type=number, default=20}
 TopRight {type=number, default=20}
 UseBackground {type=boolean, default=true}
 UseDarkInputIcon {type=boolean, default=true}
 UseIcon {type=boolean, default=true}
 UsePixels {type=boolean, default=true}
 WidthAnimation {type=number}
 WidthOverride {type=S_UI_BoolFloat, default {enable=false, value=0}}
]
widgets {
 info {inheritedWidgetCount=1, parentClass=/Script/FortniteUI.UIKitModularButton, rootWidgetClass=&SizeBox, widgetCount=8}
 widgets [
  SIzeBox {
   widgetClassPath=&SizeBox
   bIsVariable=true
   widget=SIzeBox
   properties {bOverride_MinDesiredHeight=true, bOverride_WidthOverride=true, heightOverride=50, minDesiredHeight=50, widthOverride=400}
  }
  Main {
   widgetClassPath=&GridPanel
   bIsVariable=true
   parent=SIzeBox
   widget=Main
   properties {columnFill [1], rowFill [1]}
   slot {class=SizeBoxSlot}
  }
  Content {
   widgetClassPath=&GridPanel
   bIsVariable=true
   parent=Main
   widget=Content
   properties {clipping=ClipToBounds, columnFill [0,1,0], rowFill [0,1,0]}
   slot {class=GridSlot, properties {layer=10}}
  }
  Block_EditableText {
   widgetClassPath=/FortUILibrary/Systems/InputField/Blocks/WBP_UIKit_Block_EditableText
   bIsVariable=true
   parent=Content
   widget=Block_EditableText
   slot {class=GridSlot, properties {column=1, row=1, verticalAlignment=VAlign_Center}}
  }
  Slot_Main {widgetClassPath=&NamedSlot, bIsVariable=true, parent=Main, widget=Slot_Main, slot {class=GridSlot, properties {layer=11}}}
  Block_Outline {
   widgetClassPath=/FortUILibrary/UIKit/Buttons/Blocks/WBP_UIKit_Block_Outline
   bIsVariable=true
   parent=Main
   widget=Block_Outline
   properties {
    bottomLeftCornerRadius=30
    bottomRightCornerRadius=30
    radiusInPixels=true
    topLeftCornerRadius=30
    topRightCornerRadius=30
    transitionTiming {
     deselectingFocused=0
     deselectingIdle=0
     disabling=0.1
     enabling=0.1
     focusing=0.3
     hovering=0.2
     pressing=0
     releasing=0
     selecting=0
     unfocusing=0.15
     unhovering=0.1
    }
   }
   slot {class=GridSlot, properties {layer=12}}
  }
  InputActionWidget {
   widgetClassPath=&CommonActionWidget
   bInherited=true
   bIsVariable=true
   parent=Main
   widget=InputActionWidget
   properties {visibility=Collapsed}
   slot {class=GridSlot, properties {horizontalAlignment=HAlign_Right, verticalAlignment=VAlign_Center}}
  }
  Spacer {widgetClassPath=&Spacer, bIsVariable=true, parent=Main, widget=Spacer, properties {size=<0 1>}, slot {class=GridSlot}}
 ]
}
