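# WBP_UIKit_InputField is a UI Kit input field widget Blueprint derived from WBP_UIKit_InputField_Base.
# It defines no widgets or variables of its own, and its PreConstruct event only calls the parent implementation.
# GetEditableText returns the UIKit editable text object held inside the Block_EditableText sub-widget, reached through two chained getter calls.
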WidgetBlueprint /FortUILibrary/UIKit/InputField/WBP_UIKit_InputField
graphs [
 EventGraph {nodes [
   n1 = Event PreConstruct()
   > Parent: PreConstruct(IsDesignTime=@n1.IsDesignTime)
  ]}
 GetEditableText {nodes [
   GetEditableText() {is=title K2Node_FunctionEntry}
   > Return Node!(
    execute
    NewParam:UIKit Editable Text Object Reference=Get EditableText!(self:WBP UIKit Block Editable Text Object Reference=Get Block_EditableText() -> (Block_EditableText:WBP UIKit Block Editable Text Object Reference)) -> (EditableText:UIKit Editable Text Object Reference)
   )
  ]}
]
variables []
widgets {
 info {
  inheritedWidgetCount=0
  parentClass=/FortUILibrary/Systems/InputField/WBP_UIKit_InputField_Base.WBP_UIKit_InputField_Base_C
  rootWidgetClass=null
  widgetCount=0
 }
 widgets []
}
