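# UIKit text input widget built on FortEditableFilteredCountedTextBox, wrapping WBP_UIKit_InputField with label, character count, prompt and error text.
# Typed and committed text go to the parent's handlers and are re-broadcast through dispatchers.
# The parent's dirty-text result sets icon, colors, error text and frame, plays a shake animation and optionally triggers persistent-ID format validation.
# Search mode shows a search icon and a clear button, hidden when empty or on gamepad.
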WidgetBlueprint /FortUILibrary/UIKit/InputField/WBP_UIKit_FilteredInputField
animations [Error {
  binding_count=1
  length_seconds=0.2
  object=&self:Error
  sequence {bindings [ErrorText {
     id=#1
     possessed_class=CommonTextBlock
     tracks [{
       type=MovieScene2DTransformTrack
       sections [{
         type=MovieScene2DTransformSection
         end_frame=4
         start_frame=0
         channels [
          Translation.X {type=MovieSceneScriptingFloatChannel, keys [[0,-10],[1,10],[2,-5],[3,5],[4,0]]}
          Translation.Y {type=MovieSceneScriptingFloatChannel}
          Angle {type=MovieSceneScriptingFloatChannel}
          Scale.X {type=MovieSceneScriptingFloatChannel}
          Scale.Y {type=MovieSceneScriptingFloatChannel}
          Shear.X {type=MovieSceneScriptingFloatChannel}
          Shear.Y {type=MovieSceneScriptingFloatChannel}
         ]
        }]
      }]
    }]}
 }]
graphs [
 EventGraph {nodes [
   "On Clicked (Button_Clear)"!() -> (OutputDelegate, then, Button:Common Button Base Object Reference) {is=title K2Node_ComponentBoundEvent}
   n1 = Get InputField() -> (InputField:WBP UIKit Input Field Object Reference)
   n2 = MakeLiteralText()
   > SetText!(
    execute
    self:WBP UIKit Block Editable Text Object Reference=Get Block_EditableText!(self:WBP UIKit Input Field Base Object Reference=@n1) -> (Block_EditableText:WBP UIKit Block Editable Text Object Reference)
    Text:Text=@n2
   ) -> (then)
   > OnTextChanged_Event!(execute, self:WBP UIKit Input Field Base Object Reference=@n1, Text:"Text (by ref)"=@n2) -> (then)
   n3 = "On Text Changed (InputField)"!() -> (OutputDelegate, then, Text:Text) {is=title K2Node_ComponentBoundEvent}
   > Set EditableText(EditableText:Text=@n3.Text) -> (Output_Get:Text)
   > HandleShowPrompt()
   > Call On Text Changed!(execute:Exec, self:WBP UIKit Filtered Input Field Object Reference, Text:Text=@n3.Text) -> (then:Exec) {is=title K2Node_CallDelegate}
   > HandleTextChanged!(execute, self:Fort Editable Filtered Counted Text Box Object Reference, Text:"Text (by ref)"=@n3.Text) -> (then)
   > Branch(Condition=Get IsSearchField() -> (IsSearchField:bool))
   > UpdateClearButtonVisibility()
   "On Hovered (Button_Clear)"!() -> (OutputDelegate, then, Button:Common Button Base Object Reference) {is=title K2Node_ComponentBoundEvent}
   n4 = Get Block_Outline!(self:WBP UIKit Input Field Base Object Reference=Get InputField()) -> (Block_Outline:WBP UIKit Block Outline Object Reference)
   > SetVisibility(self=@n4, InVisibility=Hidden)
   n5 = "On Text Commited (InputField)"!() -> (OutputDelegate, then, Text:Text, Commit method:ETextCommit Enum=Default) {is=title K2Node_ComponentBoundEvent}
   > HandleTextCommitted!(
    execute
    self:Fort Editable Filtered Counted Text Box Object Reference
    Text:"Text (by ref)"=@n5.Text
    CommitMethod:ETextCommit Enum=@n5.Commit method
   ) -> (then)
   > Call On Text Committed!(execute:Exec, self:WBP UIKit Filtered Input Field Object Reference, CommitMethod:ETextCommit Enum=@n5.Commit method) -> (then:Exec) {is=title K2Node_CallDelegate}
   > Branch(Condition="Equal (Enum)"(A:ETextCommit Enum=@n5.Commit method, B:ETextCommit Enum=OnEnter))
   > Call On Text Commit by Enter!(execute:Exec, self:WBP UIKit Filtered Input Field Object Reference) -> (then:Exec) {is=title K2Node_CallDelegate}
   "On Unhovered (Button_Clear)"!() -> (OutputDelegate, then, Button:Common Button Base Object Reference) {is=title K2Node_ComponentBoundEvent}
   > SetVisibility(self=@n4, InVisibility=SelfHitTestInvisible)
   "On Hovered (InputField)"!() -> (OutputDelegate, then, Button:Common Button Base Object Reference) {is=title K2Node_ComponentBoundEvent}
   > Call On Hovered!(execute:Exec, self:WBP UIKit Filtered Input Field Object Reference) -> (then:Exec) {is=title K2Node_CallDelegate}
   "On Unhovered (InputField)"!() -> (OutputDelegate, then, Button:Common Button Base Object Reference) {is=title K2Node_ComponentBoundEvent}
   > Call On Unhovered!(execute:Exec, self:WBP UIKit Filtered Input Field Object Reference) -> (then:Exec) {is=title K2Node_CallDelegate}
   n6 = InputChanged!() -> (OutputDelegate, then, Input Type:ECommonInputType Enum) {is=title K2Node_CustomEvent}
   > Set CurrentInputType(CurrentInputType:ECommonInputType Enum=@n6.Input Type) -> (Output_Get:ECommonInputType Enum)
   > UpdateClearButtonVisibility()
   n7 = Event OnDirtyTextAttempted!() -> (OutputDelegate, then, bIsBadText:bool, bAdditionalValidationFailed:bool) {is=title K2Node_Event}
   > Set Is Bad Text(Is Bad Text:bool=@n7.bIsBadText) -> (Output_Get:bool)
   > Set SanitizedTextReady(SanitizedTextReady:bool=true) -> (Output_Get:bool)
   > HandleShowPrompt()
   > n8 = Branch(Condition=Get IsSearchField())
   > n9 = UpdateClearButtonVisibility()
   "SetText (Text)"(
    execute=[@n13.then, @n9.then]
    self=Get ErrorText() -> (ErrorText:Common Text Object Reference)
    InText=GetErrorText!(Additional Validator Failed:bool=@n7.bAdditionalValidationFailed) -> (Appropriate Text:Text)
   )
   > n10 = GetCurrentText!(execute, self) -> (then, Text:Text)
   > Call On Sanitized Text Ready!(execute:Exec, self:WBP UIKit Filtered Input Field Object Reference, Text:Text=@n10, BadText:bool=@n7.bIsBadText) -> (then:Exec) {is=title K2Node_CallDelegate}
   > n11 = Branch(Condition=Get bShouldValidatePersistentID() -> (bShouldValidatePersistentID:bool))
   > n12 = ValidatePersistentID!(execute, self:Fort Editable Filtered Counted Text Box Object Reference, Text:"Text (by ref)"=@n10) -> (then)
   Branch(execute=[@n12.then, @n11.else], Condition=@n7.bIsBadText)
   > PlayAnimation(InAnimation=Get Error() -> (Error:Widget Animation Object Reference))
   SetIcon!(
    execute=@n8.else
    self:WBP UIKit Input Field Base Object Reference=Get InputField()
    Icon:Texture 2D Object Reference=Select(
     Option 0:Texture 2D Object Reference=/FortUILibrary/Textures/Icons/T_UI_Icon_UIKit_Check
     Option 1:Texture 2D Object Reference=/FortUILibrary/Textures/Icons/T_UI_Icon_UIKit_ExclamationMark
     Index=Get Is Bad Text() -> (Is Bad Text:bool)
    ) -> (ReturnValue:Texture 2D Object Reference)
   ) -> (then)
   > n13 = UpdateColors()
   Event Construct()
   n14 = Create Event()
   n15 = Get CommonInputSubsystem()
   > Bind Event to On Input Method Changed(self=@n15, Delegate=@n14)
   n16 = Get Background_Override() -> (Background_Override:Material Instance Object Reference)
   > Branch(Condition=NOT Boolean(A=IsNullItem!(ItemObject:Object Reference=@n16) -> (Is Null Item:bool)))
   > SetBackgroundMaterial!(
    execute
    self:WBP UIKit Block Background Base Object Reference=Get Block_Background!(self:WBP UIKit Input Field Base Object Reference=Get InputField()) -> (Block_Background:WBP UIKit Block Background Base Object Reference)
    Material Instance:"Material Instance Object Reference (by ref)"=@n16
   ) -> (then)
   n17 = Event OnEditableTextChanged!() -> (OutputDelegate, then, Text:"Text (by ref)") {is=title K2Node_Event}
   > SetInputFieldText!(execute, self, Text:Text=@n17.Text) -> (then)
   Event PreConstruct()
   > UpdateIconVisibility!(execute, self:WBP UIKit Input Field Base Object Reference=Get InputField(), IsVisible:bool=false) -> (then)
   > Set MaxTextCharacters!(
    execute
    MaxTextCharacters:int=GetEntryCharLimit!(self:Fort Editable Filtered Counted Text Box Object Reference) -> (ReturnValue:int=0)
    self:WBP UIKit Input Field Base Object Reference=Get InputField()
   ) -> (then, Output_Get:int=0)
   > UpdateConfiguration!(
    execute
    self
    bHideValidation:bool=Get HideValidationInfo() -> (HideValidationInfo:bool)
    TextJustification:ETextJustify Enum=Get Justification() -> (Justification:ETextJustify Enum=Left)
    bShouldShowTextPrompt:bool=Get ShowTextPrompt() -> (ShowTextPrompt:bool)
    PromptDisplayText:Text=Get PromptText() -> (PromptText:Text)
   ) -> (then)
   > Branch(Condition=Get IsSearchField())
   > SetIcon(self=Get InputField(), Icon=/FortUILibrary/Textures/Icons/T_UI_Icon_UIKit_Search)
   n18 = Get InputField()
   > Set ColorAndOpacity!(
    execute
    ColorAndOpacity:Linear Color Structure=(R=1,G=1,B=1,A=1)
    self:Image Object Reference=Get Image!(self:WBP UIKit Block Image Object Reference=Get Block_Icon!(self:WBP UIKit Input Field Base Object Reference=@n18) -> (Block_Icon:WBP UIKit Block Image Object Reference)) -> (Image:Fort Mobile Image Object Reference)
   ) -> (then, Output_Get:Linear Color Structure)
   > UpdateIconVisibility(self=Get InputField(), IsVisible=true)
   n19 = GetOwningPlayer()
   > Set CurrentInputType(CurrentInputType=GetCurrentInputType(self=Get CommonInputSubsystem!(PlayerController:Player Controller Object Reference=@n19) -> (ReturnValue:Common Input Object Reference) {is=title K2Node_GetSubsystemFromPC}))
   > UpdateClearButtonVisibility()
   > UpdateClearButtonVisuals()
   > SetGridSlotPadding!(
    execute
    Widget:Widget Object Reference=Get InputActionWidget!(self:Common Button Base Object Reference=Get InputField()) -> (InputActionWidget:Common Action Widget Object Reference)
    Padding:Margin Structure=Get InputActionPromptPadding() -> (InputActionPromptPadding:Margin Structure)
   ) -> (then)
   n20 = Event OnCharCountTextChanged!() -> (OutputDelegate, then, Text:"Text (by ref)") {is=title K2Node_Event}
   > "SetText (Text)"(self=Get Text_CharCount() -> (Text_CharCount:Common Text Object Reference), InText=@n20.Text)
   n21 = Event ValidatePersistentID!() -> (OutputDelegate, then, Text:"Text (by ref)") {is=title K2Node_Event}
   n22 = Get ErrorText()
   n23 = Select(
    Option 0:ESlateVisibility Enum=SelfHitTestInvisible
    Option 1:ESlateVisibility Enum=Collapsed
    Index=IsPersistenceIDValid!(PersistenceID:String="To String (Text)"(InText=@n21.Text)) -> (ReturnValue:bool)
   ) -> (ReturnValue:ESlateVisibility Enum)
   > Branch(Condition="Not Equal (Enum)"!(
     A:ESlateVisibility Enum=@n23
     B:ESlateVisibility Enum=GetVisibility!(self:Widget Object Reference=@n22) -> (ReturnValue:ESlateVisibility Enum=Visible)
    ) -> (ReturnValue:bool))
   > SetVisibility(self=[Get Image_WrongTextFrame() -> (Image_WrongTextFrame:Image Object Reference), Get ErrorText()], InVisibility=@n23)
   > "SetText (Text)"(self=Get ErrorText(), InText=Get PersistentIDValidationErrorText() -> (PersistentIDValidationErrorText:Text))
   Event Destruct()
   > Unbind Event from On Input Method Changed(self=@n15, Delegate=@n14)
   n24 = Event OnAddedToFocusPath!() -> (OutputDelegate, then, InFocusEvent:Focus Event Structure) {is=title K2Node_Event}
   > Parent: OnAddedToFocusPath!(execute, InFocusEvent:Focus Event Structure=@n24.InFocusEvent) -> (then) {is=title K2Node_CallParentFunction}
   > Call On Search Added to Focus Path!(execute:Exec, self:WBP UIKit Filtered Input Field Object Reference) -> (then:Exec) {is=title K2Node_CallDelegate}
   n25 = Event OnRemovedFromFocusPath!() -> (OutputDelegate, then, InFocusEvent:Focus Event Structure) {is=title K2Node_Event}
   > Parent: OnRemovedFromFocusPath!(execute, InFocusEvent:Focus Event Structure=@n25.InFocusEvent) -> (then) {is=title K2Node_CallParentFunction}
   > Call On Search Removed from Focus Path!(execute:Exec, self:WBP UIKit Filtered Input Field Object Reference) -> (then:Exec) {is=title K2Node_CallDelegate}
   SetHintText!(
    execute
    self:WBP UIKit Input Field Base Object Reference=Get InputField()
    Hint Text:Text=Event OnEditableTextHintChanged!() -> (OutputDelegate, then, Text:"Text (by ref)") {is=title K2Node_Event}.Text
   ) -> (then)
   Select(Option 0=Collapsed, Option 1=SelfHitTestInvisible, Index=@n7.bIsBadText)
   Get DefaultColor() -> (DefaultColor:Linear Color Structure)
  ]}
 GetCurrentText {nodes [
   GetCurrentText() {is=title K2Node_FunctionEntry}
   n1 = Get InputField()
   > Return Node!(
    execute
    Text:Text=Get Text(self:Editable Text Object Reference=GetEditableText!(self:WBP UIKit Input Field Object Reference=@n1) -> (NewParam:UIKit Editable Text Object Reference))
   )
  ]}
 GetEditableTextWidget {nodes [
   GetEditableTextWidget() {is=title K2Node_FunctionEntry}
   n1 = Get InputField()
   > n2 = Is Valid(InputObject=@n1)
   > Return Node!(
    execute
    ReturnValue:Editable Text Object Reference=Get EditableText!(self:WBP UIKit Block Editable Text Object Reference=Get Block_EditableText(self=@n1)) -> (EditableText:UIKit Editable Text Object Reference)
   )
   Return Node(execute=@n2.Is Not Valid)
  ]}
 GetEditedText {nodes [
   GetEditedText() {is=title K2Node_FunctionEntry}
   n1 = Get InputField()
   > Return Node(ReturnValue:Text=Get Text(self=GetEditableText(self=@n1)))
  ]}
 GetErrorText {nodes [
   GetErrorText!() -> (then, Additional Validator Failed:bool=~) {is=title K2Node_FunctionEntry}
   > Set Result(Result:Text=Get DefaultSanitizationErrorText() -> (DefaultSanitizationErrorText:Text)) -> (Output_Get:Text)
   n1 = TextIsEmpty(InText=Get AdditionalValidationErrorText() -> (AdditionalValidationErrorText:Text))
   > n2 = Branch(Condition=AND Boolean(A=Get Additional Validator Failed() -> (Additional Validator Failed:bool), B=NOT Boolean(A=@n1)))
   > n3 = Set Result(Result=Get AdditionalValidationErrorText())
   Return Node!(execute=[@n2.else, @n3.then], Appropriate Text:Text=Get Result() -> (Result:Text))
  ]}
 HandleShowPrompt {nodes [
   HandleShowPrompt() {is=title K2Node_FunctionEntry}
   > n1 = Branch(Condition=Get ShowTextPrompt())
   > SetVisibility(self=[Get ErrorText(), Get Image_WrongTextFrame()], InVisibility=Select(Option 0=Hidden, Option 1=HitTestInvisible, Index=Get Is Bad Text()))
   > SetVisibility(
    self=Get CommonTextBlock_Prompt() -> (CommonTextBlock_Prompt:Common Text Object Reference)
    InVisibility=Select(Option 0=SelfHitTestInvisible, Option 1=Hidden, Index=Get Is Bad Text())
   )
   SetVisibility(execute=@n1.else, self=Get Info() -> (Info:Overlay Object Reference), InVisibility=Collapsed)
  ]}
 OnHovered {nodes [OnHovered() {is=title K2Node_FunctionEntry}]}
 OnSanitizedTextReady {nodes [OnSanitizedTextReady!() -> (then, Text:Text, BadText:bool=~) {is=title K2Node_FunctionEntry}]}
 OnSearchAddedToFocusPath {nodes [OnSearchAddedToFocusPath() {is=title K2Node_FunctionEntry}]}
 OnSearchRemovedFromFocusPath {nodes [OnSearchRemovedFromFocusPath() {is=title K2Node_FunctionEntry}]}
 OnTextChanged {nodes [OnTextChanged!() -> (then, Text:Text) {is=title K2Node_FunctionEntry}]}
 OnTextCommitByEnter {nodes [OnTextCommitByEnter() {is=title K2Node_FunctionEntry}]}
 OnTextCommitted {nodes [OnTextCommitted!() -> (then, CommitMethod:ETextCommit Enum=Default) {is=title K2Node_FunctionEntry}]}
 OnUnhovered {nodes [OnUnhovered() {is=title K2Node_FunctionEntry}]}
 SetInputFieldText {nodes [
   n1 = SetInputFieldText!() -> (then, Text:Text) {is=title K2Node_FunctionEntry}
   > SetText(self:WBP UIKit Input Field Base Object Reference=Get InputField(), Text=@n1)
  ]}
 SetMinWidth {nodes [
   n1 = SetMinWidth!() -> (then, inWidth:double=~) {is=title K2Node_FunctionEntry}
   n2 = "Not Equal (Float)"(A=@n1)
   > Branch(Condition=@n2)
   > SetSize!(
    execute
    self:WBP UIKit Input Field Base Object Reference=Get InputField()
    WIdth:S UI Bool Float Structure=Make S UI Bool Float!(Enable_3_#2:bool=@n2, Value_4_#3:double=@n1) -> (S_UI_BoolFloat:S UI Bool Float Structure) {is=title K2Node_MakeStruct}
    Height:S UI Bool Float Structure=Get Height Override() -> (Height Override:S UI Bool Float Structure)
   ) -> (then)
  ]}
 SetReadOnly {nodes [
   n1 = SetReadOnly!() -> (then, IsReadOnly:bool=~) {is=title K2Node_FunctionEntry}
   n2 = Get InputField()
   > Set IsReadOnly!(execute, IsReadOnly:bool=@n1, self:WBP UIKit Input Field Base Object Reference=@n2) -> (then, Output_Get:bool)
   > SetReadOnly!(execute, self:WBP UIKit Block Editable Text Object Reference=Get Block_EditableText(self=@n2), IsReadOnly:bool=@n1) -> (then) {is=title K2Node_CallFunction}
  ]}
 SetTextFocus {nodes [
   SetTextFocus() {is=title K2Node_FunctionEntry}
   n1 = Get InputField()
   > SetFocus(self=Get EditableText(self=Get Block_EditableText(self=@n1)))
  ]}
 UpdateClearButtonVisibility {nodes [
   UpdateClearButtonVisibility() {is=title K2Node_FunctionEntry}
   > Branch(Condition=Get IsSearchField())
   > n1 = GetCurrentText() {is=title K2Node_CallFunction}
   n2 = TextIsEmpty(InText=@n1)
   n3 = "Equal (Enum)"(A:ECommonInputType Enum=Get CurrentInputType() -> (CurrentInputType:ECommonInputType Enum), B:ECommonInputType Enum=Gamepad)
   > n4 = Set ShowClearButton(ShowClearButton:bool=AND Boolean(A=NOT Boolean(A=@n2), B=NOT Boolean(A=@n3))) -> (Output_Get:bool)
   > n5 = Branch(Condition=@n4)
   > n6 = SetIsEnabled(self=Get Button_Clear() -> (Button_Clear:WBP UIKit Button Regular Object Reference), bInIsEnabled=true)
   SetVisibility(
    execute=[@n6.then, @n7.then]
    self=Get Button_Clear()
    InVisibility=Select(Option 0=Collapsed, Option 1=Visible, Index=Get ShowClearButton() -> (ShowClearButton:bool))
   )
   SetIsEnabled(execute=@n5.else, self=Get Button_Clear(), bInIsEnabled=false)
   > n7 = SetVisibility(self=Get Block_Outline(self=Get InputField()), InVisibility=SelfHitTestInvisible)
  ]}
 UpdateClearButtonVisuals {nodes [
   UpdateClearButtonVisuals() {is=title K2Node_FunctionEntry}
   n1 = Get Button_Clear()
   > Is Valid(InputObject=@n1)
   n2 = SlotAsOverlaySlot(Widget=@n1)
   > SetPadding!(
    execute
    self:Overlay Slot Object Reference=@n2
    InPadding:Margin Structure=Get ClearButtonPadding() -> (ClearButtonPadding:Margin Structure)
   ) -> (then)
   n3 = IsMobileGame()
   n4 = Get OverrideDefaultHeight!(self:WBP UIKit Button CTA Base Object Reference=@n1) -> (OverrideDefaultHeight:bool)
   > SetMinDimensions!(
    execute
    self:Common Button Base Object Reference=@n1
    InMinWidth:int=SelectInt(
     A=Get MobileMinWidth!(self:WBP UIKit Button CTA Base Object Reference=@n1) -> (MobileMinWidth:int=0)
     B=Get Min Width() -> (Min Width:int=0)
     bPickA=@n3
    )
    InMinHeight:int=SelectInt(
     A=Get MobileMinHeight!(self:WBP UIKit Button CTA Base Object Reference=@n1) -> (MobileMinHeight:int=0)
     B=Get Min Height() -> (Min Height:int=0)
     bPickA=AND Boolean(A=@n3, B=@n4)
    )
   ) -> (then)
   > SetVerticalAlignment(self=@n2, InVerticalAlignment=Get VerticalAlignment() -> (VerticalAlignment:EVerticalAlignment Enum=VAlign_Fill))
   n5 = Break S UI Bool Float!(S_UI_BoolFloat:"S UI Bool Float Structure (by ref)"=Get ScaleOverride() -> (ScaleOverride:S UI Bool Float Structure)) -> (Enable_3_#2:bool=~, Value_4_#3:double=~) {is=title K2Node_BreakStruct}
   > n6 = Branch(Condition=@n5)
   n7 = Get Block_Icon(self:WBP UIKit Button CTA Base Object Reference=@n1)
   > n8 = SetScalarParameterValue(
    self=Get ImageMID!(self:WBP UIKit Block Image Object Reference=@n7) -> (ImageMID:Material Instance Dynamic Object Reference)
    ParameterName=Scale
    Value=@n5.Value_4_#3
   )
   SetPadding(
    execute=[@n8.then, @n6.else]
    self:Grid Slot Object Reference=SlotAsGridSlot(Widget=@n7)
    InPadding=Get Icon Padding() -> (Icon Padding:Margin Structure)
   )
   > SetDesiredSizeOverride!(
    execute
    self:Image Object Reference=Get Image(self=@n7)
    DesiredSize:Vector 2D Structure=Get DesiredImageSize() -> (DesiredImageSize:Vector 2D Structure)
   ) -> (then)
  ]}
 UpdateColors {nodes [
   UpdateColors() {is=title K2Node_FunctionEntry}
   n1 = Get InputField()
   > Set ColorAndOpacity(
    ColorAndOpacity=Select(
     Option 0:Linear Color Structure=Get ValidColor() -> (ValidColor:Linear Color Structure)
     Option 1:Linear Color Structure=Get ErrorColor() -> (ErrorColor:Linear Color Structure)
     Index=Get Is Bad Text()
    ) -> (ReturnValue:Linear Color Structure)
    self=Get Image(self=Get Block_Icon(self:WBP UIKit Input Field Base Object Reference=@n1))
   )
   > Set!(
    execute
    ColorAndOpacity_SpecifiedColor:Linear Color Structure=Select(Option 0=Get DefaultColor(), Option 1=Get ErrorColor(), Index=Get Is Bad Text())
    ColorAndOpacity_ColorUseRule:ESlateColorStylingMode Enum
    self:Text Object Reference=Get Text_CharCount()
   ) -> (then, Output_Get:Slate Color Structure)
  ]}
 UpdateConfiguration {nodes [
   UpdateConfiguration!() -> (then, bHideValidation:bool=~, TextJustification:ETextJustify Enum=Left, bShouldShowTextPrompt:bool=~, PromptDisplayText:Text) {is=title K2Node_FunctionEntry}
   > n1 = Sequence!(execute) -> (then_0, then_1, then_2:Exec, then_3:Exec, then_4:Exec, then_5:Exec, then_6:Exec, then_7:Exec, then_8:Exec)
   Branch(execute=@n1.then_4, Condition=Get UseHintText() -> (UseHintText:bool))
   > SetHintText(self=Get InputField(), Hint Text=Get HintText() -> (HintText:Text))
   SetVisibility(
    execute=@n1.then_1
    self=[Get Info(), Get Text_CharCount()]
    InVisibility=Select(Option 0:ESlateVisibility Enum=HitTestInvisible, Option 1:ESlateVisibility Enum=Collapsed, Index=Get HideValidationInfo()) -> (ReturnValue:ESlateVisibility Enum)
   )
   Set TextJustification!(
    execute=@n1.then_2
    TextJustification:ETextJustify Enum=Get Justification() -> (Justification=Left)
    self:WBP UIKit Input Field Base Object Reference=Get InputField()
   ) -> (then, Output_Get:ETextJustify Enum=Left)
   SetVisibility(
    execute=@n1.then_0
    self=Get CommonText_Label() -> (CommonText_Label:Common Text Object Reference)
    InVisibility=Select(Option 0=HitTestInvisible, Option 1=Collapsed, Index=Get HideLabel() -> (HideLabel:bool))
   )
   > "SetText (Text)"(self=Get CommonText_Label(), InText=Get LabelText() -> (LabelText:Text))
   "SetText (Text)"(execute=@n1.then_3, self=Get CommonTextBlock_Prompt(), InText=Get PromptDisplayText() -> (PromptDisplayText:Text))
   > HandleShowPrompt() {is=title K2Node_CallFunction}
   n2 = Get HideLabel()
   n3 = Get HideValidationInfo()
   SetVisibility(
    execute=@n1.then_5
    self=Get Header() -> (Header:Horizontal Box Object Reference)
    InVisibility=Select(Option 0=HitTestInvisible, Option 1=Collapsed, Index=AND Boolean(A=@n2, B=@n3))
   )
   n4 = Get InputField()
   Set ColorAndOpacity(execute=@n1.then_6, ColorAndOpacity=Get ValidColor(), self=Get Image(self=Get Block_Icon(self=@n4)))
   > Set ColorAndOpacity(ColorAndOpacity=Get ErrorColor(), self=Get Image_WrongTextFrame())
   n5 = Set HeightOverride!(
    execute=@n1.then_7
    HeightOverride:S UI Bool Float Structure=Get Height Override()
    self:WBP UIKit Input Field Base Object Reference=Get InputField()
   ) -> (then, Output_Get:S UI Bool Float Structure)
   > n6 = Set WidthOverride!(
    execute
    WidthOverride:S UI Bool Float Structure=Get Width Override() -> (Width Override:S UI Bool Float Structure)
    self:WBP UIKit Input Field Base Object Reference=Get InputField()
   ) -> (then, Output_Get:S UI Bool Float Structure)
   > n7 = Set IconProperties!(
    execute
    IconProperties:S UI Image Structure=Get Icon Properties() -> (Icon Properties:S UI Image Structure)
    self:WBP UIKit Input Field Base Object Reference=Get InputField()
   ) -> (then, Output_Get:S UI Image Structure)
   > SetIconProperties!(execute, self:WBP UIKit Input Field Base Object Reference=Get InputField(), Data:S UI Image Structure=@n7) -> (then)
   > Set MaxExpandedWidth!(
    execute
    MaxExpandedWidth:S UI Bool Float Structure=Get MaxExpandedWidth() -> (MaxExpandedWidth:S UI Bool Float Structure)
    self:WBP UIKit Input Field Base Object Reference=Get InputField()
   ) -> (then, Output_Get:S UI Bool Float Structure)
   > n8 = Set bExpandable!(execute, bExpandable:bool=Get bExpandable() -> (bExpandable:bool), self:WBP UIKit Input Field Base Object Reference=Get InputField()) -> (then, Output_Get:bool)
   > n9 = Branch(Condition=@n8)
   n10 = Get InputField()
   > SetExpandableSize(self:WBP UIKit Input Field Base Object Reference=@n10)
   SetSize(execute=@n9.else, self=@n10, WIdth=@n6, Height=@n5)
   SetTriggeringInputAction!(
    execute=@n1.then_8
    self:Common Button Base Object Reference=Get InputField()
    InputActionRow:"Data Table Row Handle Structure (by ref)"=Get Input Action Row() -> (Input Action Row:Data Table Row Handle Structure)
   ) -> (then)
  ]}
]
variables [
 Background_Override {type=MaterialInstance, default=null}
 bExpandable {type=boolean}
 ClearButtonPadding {type=Margin, default=<ltrb 0 0 8 0>}
 CurrentInputType {type=ECommonInputType, default=MouseAndKeyboard}
 DefaultColor {type=LinearColor, default=<rgba 0.5776 0.5776 0.5776 1>}
 DesiredImageSize {type=Vector2D, default=<24 24>}
 EditableText {type=string, default=""}
 ErrorColor {type=LinearColor, default=<rgba 1 0.147 0.1329 1>}
 Height Override {type=S_UI_BoolFloat, default {enable=false, value=0}}
 HideLabel {type=boolean, default=true}
 HideValidationInfo {type=boolean}
 HintText {type=string, default=""}
 Icon Padding {type=Margin, default=<ltrb 0 0 0 0>}
 Icon Properties {
  type=S_UI_Image
  default {
   material=/FortUILibrary/Systems/InputField/MaterialInstances/MI_UIKit_InputFieldIcon
   image Size=<32 32>
   placement {
    horizontal Alignment=HAlign_Left
    vertical Alignment=VAlign_Center
    grid {column=2, column Span=0, layer=0, row=1, row Span=0, nudge=<0 0>}
    padding=<ltrb 0 0 20 0>
   }
  }
 }
 Input Action Row {type=DataTableRowHandle, default {dataTable=null, rowName=None}}
 InputActionPromptPadding {type=Margin, default=<ltrb 0 0 0 0>}
 Is Bad Text {type=boolean}
 IsSearchField {type=boolean}
 Justification {type=ETextJustify, default=Left}
 LabelText {type=string, default=None}
 MaxExpandedWidth {type=S_UI_BoolFloat, default {enable=false, value=0}}
 Min Height {type=integer, default=48}
 Min Width {type=integer, default=48}
 OnHovered {}
 OnSanitizedTextReady {}
 OnSearchAddedToFocusPath {}
 OnSearchRemovedFromFocusPath {}
 OnTextChanged {}
 OnTextCommitByEnter {}
 OnTextCommitted {}
 OnUnhovered {}
 PersistentIDValidationErrorText {type=string, default=Error: Invalid Format Text}
 PromptText {type=string, default=Numbers and symbols are allowed}
 SanitizedTextReady {type=boolean}
 ScaleOverride {type=S_UI_BoolFloat, default {enable=false, value=2}}
 ShowTextPrompt {type=boolean, default=true}
 UseHintText {type=boolean}
 ValidColor {type=LinearColor, default=<rgba 0.0395 1 0.2307 1>}
 VerticalAlignment {type=EVerticalAlignment, default=VAlign_Center}
 Width Override {type=S_UI_BoolFloat, default {enable=false, value=0}}
]
widgets {
 info {
  inheritedWidgetCount=0
  namedSlotCount=1
  parentClass=/Script/FortniteUI.FortEditableFilteredCountedTextBox
  rootWidgetClass=&GridPanel
  widgetCount=11
 }
 widgets [
  Root {widgetClassPath=&GridPanel, widget=Root, properties {columnFill [1], rowFill [0,1]}}
  Header {
   widgetClassPath=&HorizontalBox
   bIsVariable=true
   parent=Root
   widget=Header
   slot {class=GridSlot, properties {verticalAlignment=VAlign_Top, padding=<ltrb 0 0 0 16>}}
  }
  CommonText_Label {
   widgetClassPath=&CommonTextBlock
   bIsVariable=true
   parent=Header
   widget=CommonText_Label
   properties {
    mobileFontSizeMultiplier=1.2
    text=Label
    visibility=Collapsed
    colorAndOpacity {colorUseRule=UseColor_Specified, specifiedColor=<rgba 0.5776 0.5776 0.5776 1>}
    font {
     bForceMonospaced=false
     bMaterialIsStencil=false
     fontMaterial=null
     fontObject=&HeadingNow
     letterSpacing=0
     monospacedWidth=1
     size=18
     skewAmount=0
     typefaceFontName=75 Medium
     outlineSettings {
      bApplyOutlineToDropShadows=false
      bMiteredCorners=false
      bSeparateFillAlpha=false
      outlineMaterial=null
      outlineSize=0
      outlineColor=<rgba 0 0 0 1>
     }
    }
    shadowOffset=<0 0>
   }
   slot {class=HorizontalBoxSlot, properties {verticalAlignment=VAlign_Center}}
  }
  Text_CharCount {
   widgetClassPath=&CommonTextBlock
   bIsVariable=true
   parent=Header
   widget=Text_CharCount
   properties {
    mobileFontSizeMultiplier=1.5
    text=0 / 250
    colorAndOpacity {colorUseRule=UseColor_Specified, specifiedColor=<rgba 1 1 1 0.7>}
    font {
     bForceMonospaced=false
     bMaterialIsStencil=false
     fontMaterial=null
     fontObject=&HeadingNow
     letterSpacing=0
     monospacedWidth=1
     size=14.25
     skewAmount=0
     typefaceFontName=74 Regular
     outlineSettings {
      bApplyOutlineToDropShadows=false
      bMiteredCorners=false
      bSeparateFillAlpha=false
      outlineMaterial=null
      outlineSize=1
      outlineColor=<rgba 0 0 0 0>
     }
    }
    shadowColorAndOpacity=<rgba 0 0 0 0.3>
    shadowOffset=<0.5 0.5>
   }
   slot {class=HorizontalBoxSlot, properties {horizontalAlignment=HAlign_Right, verticalAlignment=VAlign_Center, size {sizeRule=Fill, value=1}}}
  }
  InputField {
   widgetClassPath=&/WBP_UIKit_InputField
   bIsVariable=true
   parent=Root
   widget=InputField
   properties {
    bShouldUseFallbackDefaultInputAction=false
    hint Text=""
    holdData=&CommonUI_HoldData_Base_C
    icon=&T_UI_Icon_UIKit_Check
    iconProperties {
     material=&MI_UIKit_InputFieldIcon
     image Size=<32 32>
     placement {
      horizontal Alignment=HAlign_Center
      vertical Alignment=VAlign_Center
      grid {column=3, column Span=0, layer=0, row=1, row Span=0, nudge=<0 0>}
      padding=<ltrb 0 0 30 0>
     }
    }
    widthOverride {enable=true, value=0}
   }
   slot {class=GridSlot, properties {row=1}}
  }
  Overlay_ButtonPlacements {widgetClassPath=&Overlay, namedSlotHost=InputField, widget=Overlay_ButtonPlacements}
  Button_Clear {
   widgetClassPath=&WBP_UIKit_Button_Regular
   bIsVariable=true
   parent=Overlay_ButtonPlacements
   widget=Button_Clear
   properties {
    bConsumePointerInput=true
    buttonSize=Small
    iconMaterial=/FortUILibrary/Textures/MaterialInstances/MI_MPLocker_Icon_Clear
    minHeight=56
    minWidth=56
    mobileMinHeight=64
    mobileMinWidth=64
    overrideDefaultHeight=true
    text=""
    touchMethod=Down
    visibility=Collapsed
   }
   slot {class=OverlaySlot, properties {horizontalAlignment=HAlign_Right, verticalAlignment=VAlign_Center, padding=<ltrb 0 0 8 0>}}
  }
  Info {
   widgetClassPath=&Overlay
   bIsVariable=true
   parent=Root
   widget=Info
   slot {class=GridSlot, properties {horizontalAlignment=HAlign_Left, row=2, verticalAlignment=VAlign_Top, padding=<ltrb 0 16 0 0>}}
  }
  ErrorText {
   widgetClassPath=&CommonTextBlock
   bIsVariable=true
   parent=Info
   widget=ErrorText
   properties {
    autoWrapText=true
    mobileFontSizeMultiplier=1.2
    text=Error: Inappropriate Language
    visibility=Hidden
    colorAndOpacity {colorUseRule=UseColor_Specified, specifiedColor=<rgba 1 0.147 0.1329 1>}
    font {
     bForceMonospaced=false
     bMaterialIsStencil=false
     fontMaterial=null
     fontObject=&HeadingNow
     letterSpacing=0
     monospacedWidth=1
     size=14.25
     skewAmount=0
     typefaceFontName=74 Regular
     outlineSettings {
      bApplyOutlineToDropShadows=false
      bMiteredCorners=false
      bSeparateFillAlpha=false
      outlineMaterial=null
      outlineSize=0
      outlineColor=<rgba 0 0 0 1>
     }
    }
    shadowOffset=<0 0>
   }
   slot {class=OverlaySlot, properties {horizontalAlignment=HAlign_Fill}}
  }
  CommonTextBlock_Prompt {
   widgetClassPath=&CommonTextBlock
   bIsVariable=true
   parent=Info
   widget=CommonTextBlock_Prompt
   properties {
    text=start typing...
    visibility=Hidden
    colorAndOpacity {colorUseRule=UseColor_Specified, specifiedColor=<rgba 1 1 1 0.7>}
    font {
     bForceMonospaced=false
     bMaterialIsStencil=false
     fontMaterial=null
     fontObject=&HeadingNow
     letterSpacing=0
     monospacedWidth=1
     size=14.25
     skewAmount=0
     typefaceFontName=74 Regular
     outlineSettings {
      bApplyOutlineToDropShadows=false
      bMiteredCorners=false
      bSeparateFillAlpha=false
      outlineMaterial=null
      outlineSize=0
      outlineColor=<rgba 0 0 0 1>
     }
    }
    shadowOffset=<0 0>
   }
   slot {class=OverlaySlot}
  }
  Image_WrongTextFrame {
   widgetClassPath=&Image
   bIsVariable=true
   parent=Root
   widget=Image_WrongTextFrame
   properties {
    visibility=Collapsed
    brush {
     bIsDynamicallyLoaded=false
     drawAs=Box
     imageType=FullColor
     mirroring=NoMirror
     resourceName=None
     resourceObject=/FortUILibrary/UIKit/Buttons/MaterialInstances/TileOutline/MI_UIKit_Outline
     tiling=NoTile
     imageSize=<64 64>
     margin=<ltrb 0.25 0.25 0.25 0.25>
     outlineSettings {
      bUseBrushTransparency=false
      roundingType=HalfHeightRadius
      width=0
      color {colorUseRule=UseColor_Specified, specifiedColor=<rgba 0 0 0 0>}
      cornerRadii=<0 0 0 0>
     }
     tintColor {colorUseRule=UseColor_Specified, specifiedColor=<rgba 1 1 1 1>}
     uVRegion {bIsValid=false, max=<0 0>, min=<0 0>}
    }
    colorAndOpacity=<rgba 1 0.147 0.1329 1>
   }
   slot {class=GridSlot, properties {layer=1, row=1, padding=<ltrb -4 0 -4 0>}}
  }
 ]
}
